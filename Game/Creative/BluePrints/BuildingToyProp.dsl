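; Creative building prop with configurable toy options, offering a "Customize" interaction to players with create permission outside play mode and vehicles.
; Option loads and changes refresh phase visibility through the VisibleInGame component and call InitializeDevice, an empty hook like OnPlayModeChanged.
; FortMinigameProgress events bind the play-mode delegate and switch health-bar suppression and, on authority, damageability.
; The creation-time damage setting is saved at minigame start and restored at end.
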(event EditorPreview|EventSetupForEditorPreview
  ;[0 1 0.1336 0.5] Initialization - Update Collision and Visuals once custom data has loaded.
  (|CallParentFunction)
  (if (Variables|Default|GetInitialize%20for%20Editor)
    (CallFunction|InitializeDevice)))

(event OnAnyPropertyChanged(ToyOptionsComponent)
  ;[0 1 0.1336 0.5] Initialization - Update Collision and Visuals once custom data has loaded.
  (CallFunction|OnPropertiesChanged)
  (CallFunction|InitializeDevice))

(event Custom|OnPropertiesChanged
  ;[0 1 0.1336 0.5] Initialization - Update Collision and Visuals once custom data has loaded.
  (if (Editor|InEditor)
    (CallFunction|EditorWorldReady)
    (else
      _))
  (Networking|SwitchHasAuthority
    (:Authority
      (Visibility|SetVisiblityDuringPhase (Variables|Default|GetVisibleInGameComponent) (VisibleinGameComponent|GetVisibilityDuringPhaseEnum))
      (CallFunction|OnPlayModeChanged (Variables|/Script/FortniteGame.FortMinigameLogicComponent|GetCurrentMinigame (Variables|Default|GetFortMinigameProgress)) (Default|IsinPlayMode)))))

(event OnPropertiesLoaded(ToyOptionsComponent)
  ;[0 1 0.1336 0.5] Initialization - Update Collision and Visuals once custom data has loaded.
  (CallFunction|OnPropertiesChanged)
  (CallFunction|InitializeDevice))

(event Custom|Initialize%20Device)

(event EventOnWorldReady
  ;[1 0 0 0.5] On World Ready - Best to setup the components for the visibility here as the static mesh component's collision is changed after BeginPlay but before here
  (VisibleinGameComponent|UpdateVisibilityInGame))

(event Custom|Editor%20World%20Ready
  ;[1 0 0 0.5] On World Ready - Best to setup the components for the visibility here as the static mesh component's collision is changed after BeginPlay but before here
  (VisibleinGameComponent|UpdateVisibilityInGame))

(event Custom|OnMinigameStarted
  (bind _bisinvulnerableduringminigameplay (Variables|Default|GetbIsInvulnerableDuringMinigamePlay))
  ;[0 1 0.05918 0.5] Minigame Started - Update whether this can be damaged during the mini game based on whether it is invulnerable during minigame play.
  (Variables|Default|SetbCanBeDamagedDuringCreationTime (Variables|Actor|GetbCanBeDamaged))
  (CallFunction|SetSuppressHealthBar _bisinvulnerableduringminigameplay)
  (Networking|SwitchHasAuthority
    (:Authority
      (Fortnite|SetCanbeDamaged self (not _bisinvulnerableduringminigameplay)))))

(event MinigameStartedDelegate(FortMinigameProgress)
  ;[0 1 0.05918 0.5] Minigame Started - Update whether this can be damaged during the mini game based on whether it is invulnerable during minigame play.
  (CallFunction|OnMinigameStarted))

(event Custom|OnMinigameEnded
  (bind _bcanbedamagedduringcreationtime (Variables|Default|GetbCanBeDamagedDuringCreationTime))
  ;[1 0.1625 0 0.5] Minigame Ended - Update whether this can be damaged during creation time.
  (CallFunction|SetSuppressHealthBar (not _bcanbedamagedduringcreationtime))
  (Networking|SwitchHasAuthority
    (:Authority
      (Fortnite|SetCanbeDamaged self _bcanbedamagedduringcreationtime))))

(event MinigameEndedDelegate(FortMinigameProgress)
  ;[1 0.1625 0 0.5] Minigame Ended - Update whether this can be damaged during creation time.
  (CallFunction|OnMinigameEnded))

(event Custom|OnPlayModeChanged
  (params (Minigame :type Fort%20Minigame%20Object%20Reference) (bIsInPlayMode :type Boolean)))

(event Custom|OnAddedToMinigame
  (params (Minigame :type Fort%20Minigame%20Object%20Reference))
  ;[1 0.75 0 0.5] Minigame Events
  (Utilities|IsValid Minigame
    (:Is%20Valid
      (Minigame|AssignOnPlayModeChangedDelegate Minigame (EventDispatchers|CreateEvent :!event Custom|OnPlayModeChanged)))))

(event OnAddedtoMinigame(FortMinigameProgress)
  (params (Minigame :type Fort%20Minigame%20Object%20Reference))
  ;[1 0.75 0 0.5] Minigame Events
  (CallFunction|OnAddedtoMinigame Minigame))

(event Custom|OnRemovedFromMinigame
  (params (Minigame :type Fort%20Minigame%20Object%20Reference))
  ;[1 0.75 0 0.5] Minigame Events
  (Utilities|IsValid Minigame
    (:Is%20Valid
      (Minigame|UnbindEventfromOnPlayModeChangedDelegate Minigame (EventDispatchers|CreateEvent :!event Custom|OnPlayModeChanged)))))

(event OnRemovedfromMinigame(FortMinigameProgress)
  (params (Minigame :type Fort%20Minigame%20Object%20Reference))
  ;[1 0.75 0 0.5] Minigame Events
  (CallFunction|OnRemovedfromMinigame Minigame))

(event EventBP_SetVisibility
  (params (bVisible :type Boolean))
  (bind _visibleingamecomponent (Variables|Default|GetVisibleInGameComponent))
  (if bVisible
    (Visibility|SetForcedVisibilityDuringGame _visibleingamecomponent "NewEnumerator1")
    (else
      (Visibility|SetForcedVisibilityDuringGame _visibleingamecomponent "NewEnumerator2"))))

(fn BlueprintCanInteract
  (params (InteractingPawn :type Fort%20Pawn%20Object%20Reference) (InteractionBeingAttempted :type EInteractionBeingAttempted%20Enum :default "FirstInteraction") (InteractionType :type TInteractionType%20Enum :default "IT_NoInteraction"))
  (bind _returnvalue (Pawn|GetController InteractingPawn))
  (Utilities|IsinVehicle InteractingPawn
    (:true
      (LocalVariables|SetbInteractSuccess false)
      (return (LocalVariables|GetbInteractSuccess)))
    (:false
      (Utilities|FlowControl|Sequence
        (:then_0
          (Utilities|IsValid (Variables|/Script/FortniteGame.FortActorOptionsComponent|GetInternal_PlayerOptions (Variables|Default|GetToyOptionsComponent))
            (:Is%20Valid
              (Utilities|IsValid InteractingPawn
                (:Is%20Valid
                  (Utilities|IsValid _returnvalue
                    (:Is%20Valid
                      (bind _asfort_player_controller_athena (Utilities|Casting|CastToFortPlayerControllerAthena _returnvalue))
                      (if (Creative|CanPlayerCreateinCurrentExperience _asfort_player_controller_athena)
                        (LocalVariables|SetbInteractSuccess (not (Default|IsinPlayMode)))))))))))
        (:then_1
          (return (LocalVariables|GetbInteractSuccess)))))))

(fn BlueprintGetInteractionString
  (params (InteractingPawn :type Fort%20Pawn%20Object%20Reference) (InteractionBeingAttempted :type EInteractionBeingAttempted%20Enum :default "FirstInteraction"))
  (return "NSLOCTEXT(\"[4ABB238642A26CA397D7E592D89218E7]\", \"DCF68A7946CED175362101BB5CEEDA3B\", \"Customize\")"))

(fn GetDesiredVisibility
  (bind _currentminigame (Variables|/Script/FortniteGame.FortMinigameLogicComponent|GetCurrentMinigame (Variables|Default|GetFortMinigameProgress))
    (:then
      (return (or (and (VisibleInGameComponent|IsVisibleDuringMinigame) (Minigame|IsinPlayMode _currentminigame :!class /Script/FortniteGame.FortMinigame)) (not (Minigame|IsinPlayMode _currentminigame :!class /Script/FortniteGame.FortMinigame)))))
    (:else
      (return (Editor|InEditor)))))

(fn GetVisibilityComponents
  (return (Utilities|Array|MakeArray (Variables|Debug|GetStaticMeshComponent))))

(fn GetVisibilityComponentsWithCollision
  (return (Utilities|Map|MakeMap)))

(fn GetVisibilityDuringPhaseEnum
  (return (select (VisibleInGameComponent|IsVisibleDuringMinigame) "NewEnumerator3" "NewEnumerator1")))

(fn IsInPlayMode
  (return (Minigame|IsinPlayMode self :!class /Script/FortniteGame.MinigameBlueprintLibrary)))

(fn IsVisibleDuringMinigame
  (return (Variables|Default|GetbIsVisibleDuringMinigame)))

(fn MinigameIsRunning
  (bind _fortminigameprogress (Variables|Default|GetFortMinigameProgress))
  (bind _currentminigame (Variables|/Script/FortniteGame.FortMinigameLogicComponent|GetCurrentMinigame _fortminigameprogress))
  (Utilities|IsValid _currentminigame
    (:Is%20Valid
      (return (Minigame|HasStarted _currentminigame :!class /Script/FortniteGame.FortMinigame)))))

(fn UpdateVisibilityinGame
  (Initialization|InitializeWithSceneComponents (Variables|Default|GetVisibleInGameComponent) (VisibleinGameComponent|GetVisibilityComponents) (VisibleinGameComponent|GetVisibilityComponentswithCollision) (VisibleinGameComponent|GetVisibilityDuringPhaseEnum)))

(fn ConstructionScript)
