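; Blueprint function library of shared AI helpers for Fortnite Creative.
; CheckAITeamValidity resolves an AI controller and pawn, then matches the offset team index against a selected team, with optional inversion.
; GetParticipantControllerFromActor resolves controllers through casts and owners, optionally rejecting guard NPCs on the test-player PlayerState branch.
; Other helpers detect guards by gameplay tag, convert options to patrol path groups, and check on-demand navmesh support using playlist tags.
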(fn CheckAITeamValidity
  (params (Actor :type Actor%20Object%20Reference) (SelectedTeam :type Creative%20Team%20Option%20Structure) (InvertTeamSelection :type Boolean))
  (bind _isaiinvalidteam (LocalVariables|GetIsAIInValidTeam))
  (bind _invertteamselection (Variables|GetInvertTeamSelection))
  (bind _returnvalue (!= _invertteamselection _isaiinvalidteam))
  (bind (_selectedteam_teamtype _selectedteam_teamindex) (Utilities|Struct|BreakCreativeTeamOption (Variables|GetSelectedTeam_TeamType)))
  (bind _asaicontroller (Utilities|Casting|CastToAIController Actor)
    (:then
      (bind _output_get (LocalVariables|SetCachedAIController _asaicontroller))
      (bind _asfort_aipawn (Utilities|Casting|CastToFortAIPawn (Pawn|GetControlledPawn _output_get))
        (:then
          (LocalVariables|SetCachedFortAIPawn _asfort_aipawn))
        (:CastFailed
          _))
      (switch Utilities|FlowControl|Switch|SwitchonECreativeTeamType _selectedteam_teamtype
        (:TeamIndex
          (LocalVariables|SetIsAIInValidTeam (== (- (Math|Conversions|ToInteger(Byte) (Team|GetActorTeam (LocalVariables|GetCachedAIController))) 2) _selectedteam_teamindex)))
        (:Any
          (LocalVariables|SetIsAIInValidTeam true)))
      (Utilities|IsValid (LocalVariables|GetCachedFortAIPawn)
        (:Is%20Valid
          (return _returnvalue (LocalVariables|GetCachedFortAIPawn) (LocalVariables|GetCachedAIController)))
        (:Is%20Not%20Valid
          (return _returnvalue 0 (LocalVariables|GetCachedAIController)))))
    (:CastFailed
      (bind _asfort_aipawn_2 (Utilities|Casting|CastToFortAIPawn Actor)
        (:then
          (bind _output_get_2 (LocalVariables|SetCachedFortAIPawn _asfort_aipawn_2))
          (LocalVariables|SetCachedAIController (AI|GetAIController _output_get_2)))
        (:CastFailed
          (return false 0 0))))))

(fn GetParticipantControllerFromActor
  (params (Actor :type Actor%20Object%20Reference) (Include%20Guard%20NPC :type Boolean :default true))
  (bind _actor (Variables|GetActor))
  (bind _returnvalue (Actor|GetOwner _actor))
  (Utilities|IsValid Actor
    (:Is%20Valid
      (bind _ascontroller (Utilities|Casting|CastToController (Variables|GetActor))
        (:then
          (LocalVariables|SetCachedController _ascontroller))
        (:CastFailed
          (bind _aspawn (Utilities|Casting|CastToPawn (Variables|GetActor))
            (:then
              (Utilities|IsValid (Pawn|GetController _aspawn)
                (:Is%20Valid
                  (LocalVariables|SetCachedController (Pawn|GetController _aspawn)))
                (:Is%20Not%20Valid
                  (return 0))))
            (:CastFailed
              (bind _asplayer_state (Utilities|Casting|CastToPlayerState (Variables|GetActor))
                (:then
                  (Utilities|IsValid (Actor|GetOwner _asplayer_state)
                    (:Is%20Valid
                      (bind _ascontroller_2 (Utilities|Casting|CastToController (Actor|GetOwner _asplayer_state))
                        (:then
                          (LocalVariables|SetCachedController _ascontroller_2)
                          (if (Class|CreativeBlueprintLibrary|IsAllowingTestPlayeronCreativeDevices)
                            (bind _asfort_athena_aibot_controller (Utilities|Casting|CastToFortAthenaAIBotController (LocalVariables|GetCachedController))
                              (:then
                                (if (not (Variables|GetInclude%20Guard%20NPC))
                                  (bind _isguard (CreativeAI|NPCType|IsGuardActor _asfort_athena_aibot_controller (Variables|GetActor)))
                                  (if _isguard
                                    (return 0)
                                    (else
                                      _))))
                              (:CastFailed
                                (Utilities|Casting|CastToPlayerController (LocalVariables|GetCachedController)
                                  (:then
                                    _)
                                  (:CastFailed
                                    _))))
                            (return (LocalVariables|GetCachedController))
                            (else
                              _)))
                        (:CastFailed
                          _)))
                    (:Is%20Not%20Valid
                      (return 0))))
                (:CastFailed
                  (Utilities|IsValid _returnvalue
                    (:Is%20Valid
                      (bind _ascontroller_3 (Utilities|Casting|CastToController _returnvalue)
                        (:then
                          (LocalVariables|SetCachedController _ascontroller_3))
                        (:CastFailed
                          _)))
                    (:Is%20Not%20Valid
                      (return 0)))))
              (return 0))))))
    (:Is%20Not%20Valid
      (return 0))))

(fn GetPatrolPathGroupFromUserOption
  (params (UserOption :type Integer))
  (if (<= UserOption 0)
    (return "Group_None")
    (else
      (return (Utilities|Enum|BytetoEnumEFortCreativePatrolPathGroup (Math|Conversions|ToByte(Integer) (- UserOption 1)))))))

(fn IsGuardActor
  (params (Actor :type Actor%20Object%20Reference))
  (bind _returnvalue (Fortnite|GetControllerfromActor Actor))
  (bind _returnvalue_2 (Pawn|GetControlledPawn _returnvalue))
  (Utilities|IsValid Actor
    (:Is%20Valid
      (Utilities|IsValid _returnvalue
        (:Is%20Valid
          (Utilities|IsValid _returnvalue_2
            (:Is%20Valid
              (if (Utilities|DoesObjectImplementInterface _returnvalue_2 "/Script/GameplayTags.GameplayTagAssetInterface")
                (return (GameplayTags|HasMatchingGameplayTag (Utilities|Casting|ToGameplayTagAssetInterface _returnvalue_2) "(TagName=\"Pawn.Athena.NPC.Creative.Guard\")"))
                (else
                  _)))
            (:Is%20Not%20Valid
              _))
          (return false))
        (:Is%20Not%20Valid
          _)))
    (:Is%20Not%20Valid
      _)))

(fn IsPlaylistSupportNavMeshOnDemand
  (params (WorldObjRef :type Object%20Reference))
  (bind _asfort_game_state_athena (Utilities|Casting|CastToFortGameStateAthena (Game|GetGameState WorldObjRef)))
  (bind _returnvalue (Navigation|IsUGCNavMeshonDemandEnabled 0))
  (return (and _returnvalue (GameplayTags|HasTag (Class|FortGameStateAthena|GetAthenaPlaylistContextTags _asfort_game_state_athena) "(TagName=\"Playlist.UGC.Core\")"))))
