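; Blueprint function library for Fortnite Creative UI that resolves icon-option textures and player display names.
; Two icon getters pick the large or small soft texture reference from a Creative Icon Option.
; If its string contains "NONE", one returns an empty reference, the other the icon picker's placeholder texture.
; A name getter casts a controller's PlayerState to FortPlayerStateAthena and returns its streaming-safe name as text, otherwise empty.
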(fn GetIconLibraryTexture
  (params (Icon%20Option :type Creative%20Icon%20Option%20Structure%20%28by%20ref%29) (Should%20Use%20Large%20Icon :type Boolean))
  (bind (_smallicon _largeicon) (Utilities|Struct|BreakCreativeIconOption Icon%20Option))
  (return (select (Utilities|String|Contains (Utilities|ToString(SoftObjectReference) (select Should%20Use%20Large%20Icon _largeicon _smallicon)) "NONE") "" (select Should%20Use%20Large%20Icon _largeicon _smallicon))))

(fn GetIconLibraryTexturewithEmptyTexture
  (params (Icon%20Option :type Creative%20Icon%20Option%20Structure%20%28by%20ref%29) (Should%20Use%20Large%20Icon :type Boolean))
  (bind (_smallicon _largeicon) (Utilities|Struct|BreakCreativeIconOption Icon%20Option))
  (return (select (Utilities|String|Contains (Utilities|ToString(SoftObjectReference) (select Should%20Use%20Large%20Icon _largeicon _smallicon)) "NONE") "/Game/Athena/UI/InGame/Creative/PropertyEditor/Icon_Picker/Textures/T_UI_IconPicker_EMPTY.T_UI_IconPicker_EMPTY" (select Should%20Use%20Large%20Icon _largeicon _smallicon))))

(fn GetLocalPlayerNamefromController
  (params (Controller :type Controller%20Object%20Reference))
  (Utilities|IsValid Controller
    (:Is%20Valid
      (bind _asfort_player_state_athena (Utilities|Casting|CastToFortPlayerStateAthena (Variables|/Script/Engine.Controller|GetPlayerState Controller))
        (:then
          (return (Utilities|Text|ToText(String) (PlayerState|GetPlayerNameforStreaming _asfort_player_state_athena))))
        (:CastFailed
          _)))
    (:Is%20Not%20Valid
      _))
  (return ""))
