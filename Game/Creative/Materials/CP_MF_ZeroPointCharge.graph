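# Material function that layers a Zero Point energy charge onto input material attributes, adding emissive glow and darkening base color.
# Vertex-interpolated pre-skinned distance masks grow the charge from a point with EnergyProgress and carve out a second location.
# Inside, distorted panning electric noise remaps chromatic smoke lookups, combined with HeightLerp breakup, Fresnel rim and sine rings shifted by EnergyDirection.
# Progress, subtract, direction and brightness inputs default to the CP_MPC_Doggus parameter collection.
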MaterialFunction /Game/Creative/Materials/CP_MF_ZeroPointCharge
graph {nodes [
  Constant3Vector() {constant=<rgba -69.845 0 143.5788 1>}
  Constant() {r=1}
  n1 = GetMaterialAttributes(None=FunctionInput() {inputType=FunctionInput_MaterialAttributes}) {attributeGetTypes [#1, #2, #3]}
  Fresnel(Normal=@n1.Normal)
  FunctionInput() {inputName=Mask, inputType=FunctionInput_Scalar, sortPriority=3}
  n2 = Fresnel(ExponentIn=ScalarParameter() {defaultValue=2.8424, parameterName=CromaticExponent}, Normal=Normalize(VectorInput=Transform(None=@n1.Normal))) {baseReflectFraction=0}
  n3 = VectorParameter() {parameterName=ZeroPointColorTint_Color2, defaultValue=<rgba 0.1406 0.4879 1 1>}
  n4 = Time()
  n5 = VectorParameter() {parameterName=Energy_Speed, defaultValue=<rgba 0.05 0.02 0.05 0.05>}
  n6 = Panner(
   Coordinate=Multiply(A=TextureCoordinate(), B=ScalarParameter() {defaultValue=3.0931, parameterName=Energy_Tile1})
   Time=@n4
   Speed=MaterialFunctionCall(X=@n5.R, Y=@n5.R) {materialFunction=&MakeFloat2}.Result
  )
  n7 = Multiply(A=TextureCoordinate(), B=ScalarParameter() {defaultValue=8, parameterName=Energy_DistortTile})
  n8 = Multiply(
   A=Time()
   B=MaterialFunctionCall(X=ScalarParameter() {defaultValue=0.2, parameterName=Energy_DistortSpeed}, Y=ScalarParameter() {parameterName=Energy_DistortSpeed2}) {materialFunction=&MakeFloat2}.Result
  )
  n9 = TextureSample(UVs=Frac(None=Add(A=@n7, B=@n8))) {samplerType=SAMPLERTYPE_LinearColor, texture=/Game/Missions/Primary/DeliverTheBomb/Textures/T_CloudNoise}
  n10 = Multiply(A=Subtract(A=AppendVector(A=@n9.R, B=@n9.R)) {constB=0.5}, B=ScalarParameter() {defaultValue=0.0381, parameterName=Energy_DistortAmount}) {constB=0.2}
  n11 = TextureSample(UVs=TextureSample(UVs=Add(A=@n6, B=@n10)) {samplerType=SAMPLERTYPE_LinearColor, texture=/Game/Characters/Player/Male/Medium/Bodies/M_MED_SRDrift/Textures/T_ElectricNoise}.R) {samplerSource=SSM_Wrap_WorldGroupSettings, texture=&T_Inky_Smoke_Tile_Chromatic}
  n12 = Power(
   Base=Multiply(A=@n11.RGB, B=ScalarParameter() {defaultValue=800, parameterName=Brigthness Electricity})
   Exp=ScalarParameter() {defaultValue=0.8129, parameterName=PowerElectricity}
  )
  n13 = VectorParameter() {parameterName=ZeroPointColorTint_Color, defaultValue=<rgba 0.328 0.7872 1 1>}
  n14 = Panner(
   Coordinate=Multiply(A=TextureCoordinate(), B=ScalarParameter() {defaultValue=3.0498, parameterName=Energy_Tile2})
   Time=@n4
   Speed=MaterialFunctionCall(X=@n5.B, Y=@n5.B) {materialFunction=&MakeFloat2}.Result
  )
  n15 = TextureSample(UVs=TextureSample(UVs=Add(A=@n10, B=@n14)) {samplerType=SAMPLERTYPE_LinearColor, texture=&T_ElectricNoise}.R) {samplerSource=SSM_Wrap_WorldGroupSettings, texture=&T_Inky_Smoke_Tile_Chromatic}
  n16 = Power(
   Base=Multiply(A=@n15.R, B=ScalarParameter() {defaultValue=800, parameterName=Brigthness Electricity})
   Exp=ScalarParameter() {defaultValue=0.8129, parameterName=PowerElectricity}
  )
  n17 = Multiply(A=@n2, B=Add(A=Multiply(A=@n3.RGB, B=@n12), B=Multiply(A=@n13.RGB, B=@n16)))
  n18 = Distance(A=PreSkinnedPosition(), B=VectorParameter() {parameterName=LocationZeroPointSubtractMask, defaultValue=<rgba 60 0 0 1>}.RGB)
  n19 = LinearInterpolate(
   A=Constant() {r=0}
   B=ScalarParameter() {defaultValue=140.9162, parameterName=DistanceSubtractMask}
   Alpha=FunctionInput(Preview=CollectionParameter() {collection=&/CP_MPC_Doggus, parameterName=Energy_SubtractMaskDistance}) {bUsePreviewValueAsDefault=true, inputName=EnergySubtractMask, inputType=FunctionInput_Scalar, sortPriority=4}
  )
  n20 = Power(Base=Saturate(None=Divide(A=@n18, B=@n19)), Exp=ScalarParameter() {defaultValue=30, parameterName=DistanceSubtractMaskFalloff}) {constExponent=3}
  n21 = VectorParameter() {parameterName=LocationZeroPoint, defaultValue=<rgba -97.4333 0 96.378 1>}
  n22 = Distance(A=PreSkinnedPosition(), B=@n21.RGB)
  n23 = LinearInterpolate(
   A=Constant() {r=0}
   B=ScalarParameter() {defaultValue=108.1562, parameterName=DistanceBlendZero}
   Alpha=FunctionInput(Preview=CollectionParameter() {collection=&CP_MPC_Doggus, parameterName=Energy_Progress}) {bUsePreviewValueAsDefault=true, inputName=EnergyProgress, inputType=FunctionInput_Scalar, sortPriority=4}
  )
  n24 = OneMinus(None=Saturate(None=Divide(A=@n22, B=@n23)))
  n25 = ScalarParameter() {defaultValue=1.4462, parameterName=DistanceBlendZeroFalloff}
  n26 = Multiply(A=Saturate(None=VertexInterpolator(VS=@n20).PS), B=VertexInterpolator(VS=Power(Base=@n24, Exp=@n25) {constExponent=3}).PS)
  n27 = Add(A=@n11.RGB, B=@n15.R)
  n28 = Multiply(A=Multiply(
    A=@n17
    B=MaterialFunctionCall(Transition Phase=@n26, Height Texture=@n27) {materialFunction=&HeightLerp}.Alpha
   ))
  n29 = ComponentMask(None=@n26) {r=true}
  n30 = FunctionInput(Preview=CollectionParameter() {collection=&CP_MPC_Doggus, parameterName=Energey_Direction}) {bUsePreviewValueAsDefault=true, inputName=EnergyDirection, inputType=FunctionInput_Scalar, sortPriority=4}
  n31 = Add(A=Divide(A=@n22, B=ScalarParameter() {defaultValue=12.3011, parameterName=WaveLength}), B=@n30)
  n32 = Power(Base=Add(A=Sine(None=@n31)), Exp=ScalarParameter() {defaultValue=4, parameterName=WavePower})
  n33 = ScalarParameter() {defaultValue=1, parameterName=WaveBrightness}
  n34 = ComponentMask(None=VertexInterpolator(VS=Multiply(A=@n32, B=@n33)).PS) {r=true}
  n35 = OneMinus(None=Multiply(A=ScalarParameter() {defaultValue=1, parameterName=Chromatics}, B=@n2))
  n36 = Power(
   Base=Multiply(
    A=TextureSample(UVs=@n35) {samplerSource=SSM_Wrap_WorldGroupSettings, texture=&T_Inky_Smoke_Tile_Chromatic}.RGB
    B=@n2
   )
   Exp=ScalarParameter() {defaultValue=0.9039, parameterName=FresnelPowerZero}
  )
  n37 = ScalarParameter() {defaultValue=14.8999, parameterName=FresnelBrigthnessZero}
  n38 = Add(A=Multiply(A=@n29, B=Multiply(A=@n34, B=@n17)), B=Multiply(A=@n29, B=Multiply(A=@n36, B=@n37)))
  n39 = CollectionParameter() {collection=&CP_MPC_Doggus, parameterName=Energy_Brightness}
  n40 = Add(
   A=@n1.EmissiveColor
   B=Multiply(
    A=Add(A=@n28, B=@n38)
    B=FunctionInput(Preview=@n39) {bUsePreviewValueAsDefault=true, inputName=EnergyBrightness, inputType=FunctionInput_Scalar, sortPriority=4}
   )
  )
  n41 = Multiply(A=@n1.BaseColor, B=VectorParameter() {parameterName=DarkenBaseColor, defaultValue=<rgba 0 0.0195 0.0781 1>}.RGB)
  n42 = Saturate(None=Multiply(A=ScalarParameter() {defaultValue=1.2986, parameterName=DarkenIntensity}, B=@n26))
  FunctionOutput(None=SetMaterialAttributes!(
    MaterialAttributes=@n1.MaterialAttributes
    EmissiveColor=LinearInterpolate(A=@n1.EmissiveColor, B=@n40, Alpha=@n26)
    BaseColor=LinearInterpolate(A=@n1.BaseColor, B=@n41, Alpha=@n42)
   ) {attributeSetTypes [#1, #2]})
  Multiply()
  Multiply(
   A=Multiply(A=@n2, B=OneMinus(None=@n27))
   B=VectorParameter() {parameterName=ZeroPointColorTint_Color3, defaultValue=<rgba 1.6979 0.7146 5 1>}.RGB
  )
  Multiply(A=Time(), B=LinearInterpolate(A=Constant(), B=ScalarParameter() {defaultValue=-0.7758, parameterName=SineSpeed}, Alpha=@n30))
  Multiply(A=@n2, B=ScalarParameter() {defaultValue=0.0286, parameterName=FBrightness})
  ScalarParameter() {parameterName=Energy_Amount}
  ScalarParameter() {defaultValue=0.7758, parameterName=SineSpeed2}
  TextureSample() {samplerType=SAMPLERTYPE_LinearGrayscale, texture=&T_ElectricNoise}
  TransformPosition(None=@n21.RGB) {transformType=TRANSFORMPOSSOURCE_World}
 ]}
