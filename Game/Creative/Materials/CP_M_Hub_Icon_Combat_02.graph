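# Unlit, masked material for the Creative hub's Combat icon, sampling the CP_Hub_Icon_Combat_02 texture.
# A Scalar_Time_of_Day function supplies a brightness factor that multiplies the texture RGB, brightest by day and dimmest at night.
# A separate lerp with placeholder-labeled constants and several loose constants are also present, with no material output links shown.
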Material /Game/Creative/Materials/CP_M_Hub_Icon_Combat_02
graph {
 properties {bCastRayTracedShadows=false, blendMode=BLEND_Masked, pixelDepthOffsetMode=PDOM_Legacy, shadingModel=MSM_Unlit, usedShadingModels=Unlit}
 nodes [
  Constant() {r=0.2}
  Constant() {r=1}
  Constant()
  n1 = TextureSample() {texture=/Game/Creative/Textures/CP_Hub_Icon_Combat_02}
  LinearInterpolate(A=@n1.RGB, B=Constant() {desc=Texture here}, Alpha=Constant() {desc=Texture alpha here})
  n2 = Constant() {r=0.4}
  Multiply(
   A=@n1.RGB
   B=MaterialFunctionCall(Value_Morning=@n2, Value_Day=Constant() {r=0.7}, Value_Evening=@n2, Value_Night=Constant() {r=0.3}) {materialFunction=&Scalar_Time_of_Day}.Result
  )
 ]
}
