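# Creative-mode foliage master material (masked, two-sided foliage) that recolors a diffuse texture with a picked color.
# Desaturated, contrast-adjusted diffuse multiplies the color, with a luminance-based mask blending in a hue-shifted variant, previewable via static switch.
# Specular and roughness share one mask channel; opacity comes from diffuse or a mask texture through a selectable channel.
# Attributes then chain through tree subsurface, two wind functions and the Creative move-tool function.
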Material /Game/Creative/Materials/M_CP_Flora_ColorPicker
graph {
 properties {
  bUseMaterialAttributes=true
  blendMode=BLEND_Masked
  pixelDepthOffsetMode=PDOM_Legacy
  shadingModel=MSM_TwoSidedFoliage
  twoSided=true
  usedShadingModels=Two Sided Foliage
 }
 nodes [
  n1 = TextureSampleParameter2D() {group=Textures, parameterName=Diffuse, texture=/Game/Environments/Props/Transylvania/Trees/Textures/T_TRV_LeafPile_D}
  n2 = VectorParameter() {group=Color, parameterName=Color, sortPriority=1, defaultValue=<rgba 0.5234 0.1408 0.0438 1>}
  n3 = MaterialFunctionCall(Hue Shift Percentage=ScalarParameter() {group=Color, parameterName=Hueshift, sliderMax=1, sortPriority=40}, Texture=@n2.RGB) {materialFunction=&HueShift}
  n4 = Desaturation(None=@n1.RGB, Fraction=Constant() {r=1})
  n5 = ScalarParameter() {group=Color, parameterName=Hueshift Mask Contrast, sortPriority=45}
  n6 = Saturate(None=Multiply(
    A=MaterialFunctionCall(In=@n4, Contrast=@n5) {materialFunction=&CheapContrast}.Result
    B=ScalarParameter() {defaultValue=1, group=Color, parameterName=Hueshift Intensity, sortPriority=45}
   ) {constB=50})
  n7 = Multiply(A=@n4, B=ScalarParameter() {defaultValue=2, group=Color, parameterName=Birghtness Mult for ColorPicker, sortPriority=2}) {constB=2}
  n8 = ScalarParameter() {group=Color, parameterName=Diffuse Texture Contrast, sortPriority=3}
  n9 = LinearInterpolate(
   A=@n1.RGB
   B=Multiply(
    A=LinearInterpolate(A=@n2.RGB, B=@n3.Result, Alpha=@n6) {desc=Blend in Hueshift}
    B=MaterialFunctionCall(In=@n7, Contrast=@n8) {materialFunction=&CheapContrast}.Result
   )
   Alpha=ScalarParameter() {defaultValue=1, desc=Dynamic on/off gradient, group=Color, parameterName=HueshiftOpacity, sliderMax=1}
  )
  n10 = TextureSampleParameter2D() {
   desc="spec, metallic,roughness\n"
   group=Textures
   parameterName=SpecularMasks
   samplerType=SAMPLERTYPE_Masks
   texture=/Game/Environments/Props/Transylvania/Trees/Textures/T_TRV_LeafPile_S
  }
  n11 = TextureSampleParameter2D() {
   group=Masks
   parameterName=MaskTexture
   samplerType=SAMPLERTYPE_Masks
   texture=/Game/Environments/Props/Transylvania/Trees/Textures/T_TRV_LeafPile_M
  }
  n12 = ChannelMaskParameter(None=StaticSwitchParameter(True=@n11.RGB, False=@n1.RGBA) {group=Masks, parameterName=Use MaskTexture Opacity, sortPriority=2}) {group=Masks, maskChannel=Alpha, parameterName=Opacity Mask Channel, defaultValue=<rgba 0 0 0 1>}
  n13 = TextureSampleParameter2D() {
   group=Textures
   parameterName=Normals
   samplerType=SAMPLERTYPE_Normal
   texture=/Game/Environments/Props/Transylvania/Trees/Textures/T_TRV_LeafPile_N
  }
  n14 = TextureSampleParameter2D() {
   group=Masks
   parameterName=SubSurface Mask
   samplerType=SAMPLERTYPE_Masks
   texture=&T_TRV_LeafPile_S
  }
  n15 = MaterialFunctionCall!(
   In=MakeMaterialAttributes!(
    BaseColor=StaticSwitchParameter(True=@n6, False=@n9) {group=Color, parameterName=Preview HueShift Mask, sortPriority=50}
    Metallic=Constant()
    Specular=@n10.R
    Roughness=@n10.R
    Anisotropy
    EmissiveColor
    Opacity=@n12.R
    OpacityMask=@n12.R
    Normal=@n13.RGB
    Tangent
    WorldPositionOffset
    SubsurfaceColor
    ClearCoat
    ClearCoatRoughness
    AmbientOcclusion
    Refraction
    CustomizedUVs_0
    CustomizedUVs_1
    CustomizedUVs_2
    CustomizedUVs_3
    CustomizedUVs_4
    CustomizedUVs_5
    CustomizedUVs_6
    CustomizedUVs_7
    PixelDepthOffset
    ShadingModel
    Displacement
   )
   Mask=ChannelMaskParameter(None=@n14.RGBA) {group=Masks, parameterName=SubSurface Mask Channel}.R
  ) {materialFunction=/Game/Packages/Fortnite_UniqueMaterials_1/MaterialFunctions/Trees/MF_Tree_SubSurface}
  MaterialFunctionCall(MA In=MaterialFunctionCall!(In=MaterialFunctionCall!(In=@n15.Result) {materialFunction=/Game/Athena/WinterArt/Materials/MF_WPO_WindNew}.Result) {materialFunction=/Game/Athena/WinterArt/Materials/MF_WindWPO}.Result) {materialFunction=&MF_CreativeMoveTool}
 ]
}
