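# Masked two-sided foliage master material for Creative grass and small plants, recolored through color picker parameters, with Nanite and instancing support.
# It chains Apollo small-foliage functions from swamp grass to jungle plants, then applies snow progression and Curie landscape functions.
# The resulting base color is desaturated, lifted by a minimum brightness and tinted.
# UV regions select the grass color or its hue-shifted, brightness-scaled variants for flowers and ferns/weeds.
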Material /Game/Creative/Materials/M_CP_Grass_ColorPicker
graph {
 properties {
  bTangentSpaceNormal=false
  bUseMaterialAttributes=true
  bUsedWithInstancedStaticMeshes=true
  bUsedWithNanite=true
  blendMode=BLEND_Masked
  pixelDepthOffsetMode=PDOM_Legacy
  shadingModel=MSM_TwoSidedFoliage
  twoSided=true
  usedShadingModels=Two Sided Foliage
 }
 nodes [
  OneMinus(None=TextureSampleParameter2D() {parameterName=MaskTexture, texture=/Game/Athena/Apollo/Environments/Landscape/SmallFoliage/Assets/BasicGrass/T_ApolloGrass01}.R)
  n1 = StaticBool()
  n2 = StaticBoolParameter() {defaultValue=true, desc="Disconnect this from \"IsStaticSnow\" when using dynamic snow", parameterName=Snow-IsForGrassMeshToPushDown}
  n3 = StaticBool()
  n4 = StaticBoolParameter() {parameterName=IsForLobby}
  n5 = StaticBool()
  n6 = SetMaterialAttributes!(MaterialAttributes)
  n7 = MaterialFunctionCall!(In=MaterialFunctionCall!(In=MaterialFunctionCall!(In=@n6) {materialFunction=/Game/Athena/Apollo/Environments/Landscape/SmallFoliage/MasterMaterial/MaterialFunctions/MF_SwampGrass}.Result) {materialFunction=/Game/Athena/Apollo/Environments/Landscape/SmallFoliage/MasterMaterial/MaterialFunctions/MF_Heather}.Result) {materialFunction=/Game/Athena/Apollo/Environments/Landscape/SmallFoliage/MasterMaterial/MaterialFunctions/MF_BasicGrass}
  n8 = MaterialFunctionCall!(In=MaterialFunctionCall!(In=MaterialFunctionCall!(In=@n7.Result) {materialFunction=/Game/Athena/Apollo/Environments/Landscape/SmallFoliage/MasterMaterial/MaterialFunctions/MF_LGRock}.Result) {materialFunction=/Game/Athena/Apollo/Environments/Landscape/SmallFoliage/MasterMaterial/MaterialFunctions/MF_Leaves}.Result) {materialFunction=/Game/Athena/Apollo/Environments/Landscape/SmallFoliage/MasterMaterial/MaterialFunctions/MF_JunglePlants}
  n9 = GetMaterialAttributes(None=MaterialFunctionCall!(
    In=MaterialFunctionCall!(
     IsStaticSnow=@n1
     UseAlternateStaticSnowMask
     AlternateStaticSnowMask
     "UseRoad&&OcclusionMask"
     IsForGrassMeshToPushDown=@n2
     AdditionalMaskToRemoveSnow
     IsFallingSnowFX
     UseMinimapCoords
     IncreaseSnowUsingSlope=@n3
     UseSmoothBorder
     LerpHighM
     TemporalHeightLerpContrast
     IsLobbyIsland=@n4
     UseBR-IslandMask
     IsTree
     IsForRenderTarget
     UseSparkle
     UseSnowNormals
     UseAdditionalMaskToRemoveSnow
     IsTangentSpaceInput
     UseVertexNormalsForBasicSnow
     UseTangentSnowNormals=@n5
     Material=@n8.Result
    ) {materialFunction=/Game/Athena/Apollo/Environments/Landscape/Materials/Biome/UniqueBiomeMaterialFunctions/MF_Apollo_Snow_Progression}.Material
    UseHeightFalloff=StaticBool()
    OnFireEmissiveOverride
    HeightFalloffRadius
    HeightFalloffHardness
    VariationMasksMobile
    In_Grass
    Is_Stylised_WE
   ) {materialFunction=/Game/Packages/Fortress_SharedMaterials/Base_Material_Functions/MF_Curie_Landscape}.Result) {attributeGetTypes [#1]}
  n10 = ScalarParameter() {desc="Bring  up the darker values", group=ColorPicker, parameterName=Min Bightness}
  n11 = Desaturation(None=@n9.Base Color)
  n12 = MaterialFunctionCall(Float2=TextureCoordinate()) {materialFunction=&BreakOutFloat2Components}
  n13 = VectorParameter() {group=ColorPicker, parameterName=Grass Color, sortPriority=5, defaultValue=<rgba 0.1738 0.375 0.1953 1>}
  n14 = Multiply(
   A=MaterialFunctionCall(
    Hue Shift Percentage=ScalarParameter() {defaultValue=0.5, group=ColorPicker, parameterName=Flowers-HueShift, sortPriority=20}
    Texture=@n13.RGB
   ) {materialFunction=&HueShift}.Result
   B=ScalarParameter() {defaultValue=1, group=ColorPicker, parameterName=Flowers-Brightness, sortPriority=21}
  )
  n15 = MaterialFunctionCall(
   Hue Shift Percentage=ScalarParameter() {defaultValue=-0.1, group=ColorPicker, parameterName=FernsWeeds-HueShift, sortPriority=10}
   Texture=@n13.RGB
  ) {materialFunction=&HueShift}
  n16 = ScalarParameter() {defaultValue=1, group=ColorPicker, parameterName=FernsWeeds-Brightness, sortPriority=11}
  n17 = If(
   A=OneMinus(None=@n12.G)
   B=Constant() {r=0.6}
   "A > B"=@n13.RGB
   "A == B"=@n13.RGB
   "A < B"=If(A=@n12.R, B=Constant() {r=0.6}, "A > B"=@n14, "A == B"=@n14, "A < B"=Multiply(A=@n15.Result, B=@n16))
  )
  n18 = ScalarParameter() {defaultValue=1, desc=Tuneable to pop from ground grass cover, group=ColorPicker, parameterName=Overall Brightness}
  SetMaterialAttributes!(
   MaterialAttributes=@n9.MaterialAttributes
   Base Color=Multiply(A=LinearInterpolate(A=@n10, Alpha=@n11) {constA=0.25}, B=Multiply(A=@n17, B=@n18))
  ) {attributeSetTypes [#1]}
 ]
}
