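# Creative-mode ground material that layers terrain surfaces and lets grass and ground be tinted by color parameters.
# Vertex-color blue and height lerps drive blend masks for grass, gravel and mud or sand, with static switches choosing layers or a custom-material-only path.
# Recoloring desaturates base color, then multiplies by the chosen color, with optional masked hue shift on grass.
# Forward and mobile paths use vertex normals; the switched result feeds MF_CreativeMoveTool.
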Material /Game/Creative/Materials/M_CP_Ground_ColorPicker
graph {
 properties {
  bTangentSpaceNormal=false
  bUseMaterialAttributes=true
  ditheredLODTransition=true
  pixelDepthOffsetMode=PDOM_Legacy
  usedShadingModels=Default Lit
 }
 nodes [
  n1 = MaterialFunctionCall() {materialFunction=/Game/Athena/Apollo/Environments/Terrain_Meshes/MaterialFunctions/MF_Common_GenericMaterial}
  n2 = VertexNormalWS()
  n3 = StaticSwitchParameter(True=BlendMaterialAttributes(A=NamedRerouteUsage(), B=NamedRerouteUsage(), Alpha=NamedRerouteUsage()), False=NamedRerouteUsage()) {
   defaultValue=true
   desc="If True, then Grass will be shown where the vertex color is black."
   group=Layers
   parameterName=UseGrass
   sortPriority=5
  }
  n4 = SetMaterialAttributes!(MaterialAttributes=@n3, Normal=ShadingPathSwitch(Default=@n3, Forward=@n2, Mobile=@n2)) {
   desc="This is assuming there are no normal textures on mobile other than the Cliffs && Custom Material"
   attributeSetTypes [#1]
  }
  MaterialFunctionCall(MA In=StaticSwitchParameter(True=@n1.Result, False=@n4) {desc="If set to true, no other layers will be used.", group=Layers, parameterName=UseCustomMaterialOnly, sortPriority=-10}) {materialFunction=&MF_CreativeMoveTool}
  MaterialFunctionCall!(CustomMaterial=@n1.Result, CustomMaterial_Height=@n1.Result) {
   desc=Stripped Parts from this for Control
   materialFunction=/Game/Athena/Apollo/Environments/Terrain_Meshes/MaterialFunctions/MF_VextexColorLayerBlend
  }
  MaterialFunctionCall!(
   Base Material=NamedRerouteUsage()
   Top Material=NamedRerouteUsage() {desc=Ground Material}
   Alpha=MaterialFunctionCall(
    Transition Phase=VertexColor().B
    Height Texture=NamedRerouteUsage()
    Contrast=ScalarParameter() {defaultValue=0.4595, parameterName=GrassBlendContrast}
   ) {materialFunction=&HeightLerp}.Alpha
  ) {materialFunction=/Engine/Functions/MaterialLayerFunctions/MatLayerBlend_Standard}
  n5 = GetMaterialAttributes(None=MaterialFunctionCall() {materialFunction=/Game/Athena/Environments/Landscape/MaterialFunctions/Standard/MF_Athena_Grass_01}.Result) {attributeGetTypes [#2]}
  n6 = Desaturation(None=@n5.Base Color)
  n7 = ScalarParameter() {group=Grass, parameterName=Grass Texture Contrast, sortPriority=41}
  n8 = ScalarParameter() {defaultValue=2, group=Grass, parameterName=Birghtness Mult for ColorPicker, sortPriority=40}
  n9 = VectorParameter() {parameterName=Grass Color, defaultValue=<rgba 0.1738 0.375 0.1953 1>}
  n10 = MaterialFunctionCall(In=@n6, Contrast=ScalarParameter() {defaultValue=0.15, group=Grass, parameterName=Hueshift Mask Contrast, sortPriority=45}) {materialFunction=&CheapContrast}
  n11 = ScalarParameter() {defaultValue=1, group=Grass, parameterName=Hueshift Intensity, sortPriority=45}
  n12 = LinearInterpolate(
   A=@n9.RGB
   B=MaterialFunctionCall(Hue Shift Percentage=ScalarParameter() {group=Grass, parameterName=Hueshift, sliderMax=1, sortPriority=45}, Texture=@n9.RGB) {materialFunction=&HueShift}.Result
   Alpha=Saturate(None=Multiply(A=@n10.Result, B=@n11) {constB=50})
  )
  n13 = SetMaterialAttributes!(
   MaterialAttributes=@n5.MaterialAttributes
   Base Color=Multiply(
    A=MaterialFunctionCall(In=@n6, Contrast=@n7) {materialFunction=&CheapContrast}.Result
    B=Multiply(A=@n8, B=@n12) {constB=2}
   )
  ) {attributeSetTypes [#2]}
  NamedRerouteDeclaration(None=@n13) {name=Grass Material New, nodeColor=<rgba 1 0 0.8203 1>}
  n14 = NamedRerouteUsage() {desc=Ground Material}
  n15 = Desaturation(None=GetMaterialAttributes(None=@n14) {attributeGetTypes [#2]}.Base Color)
  n16 = VectorParameter() {parameterName=Ground Color, defaultValue=<rgba 0.6953 0.5069 0.3071 1>}
  n17 = SetMaterialAttributes(MaterialAttributes=@n14, Base Color=Multiply(A=Multiply(A=@n15, B=@n16.RGB)) {constB=2}) {attributeSetTypes [#2]}
  NamedRerouteDeclaration(None=@n17) {name=Ground Recolored, nodeColor=<rgba 1 0.504 0 1>}
  NamedRerouteDeclaration(None=@n4) {name=Possible-GrassGravelMud/SandAsphalt, nodeColor=<rgba 1 0 0.9844 1>}
  n18 = MaterialFunctionCall!(
   OverrideEdgeColor=MaterialFunctionCall() {materialFunction=/Game/Athena/Apollo/Environments/Landscape/Materials/Biome/CommonMaterials/MF_GravelAndMudColors}.SM-GravelEdgeColorMod
   EdgeMask=NamedRerouteUsage()
   BypassBiomeColors
  ) {
   desc="Gravel used to be referred to as \"Dirt\" in the legacy terrain material"
   materialFunction=/Game/Athena/Apollo/Environments/Landscape/Materials/Biome/CommonMaterials/Mobile/MF_Mobile_Common_Gravel
  }
  NamedRerouteDeclaration(None=@n18.ForStaticMesh) {name=Gravel-Material, nodeColor=<rgba 1 0 0.9844 1>}
  n19 = GetMaterialAttributes(None=MaterialFunctionCall!(CustomSeaLevel, BypassBiomeColors, UseOceanLevel) {materialFunction=/Game/Athena/Apollo/Environments/Landscape/Materials/Biome/CommonMaterials/Mobile/MF_Mobile_Common_Sand}.ForStaticmesh) {attributeGetTypes [#2]}
  NamedRerouteDeclaration(None=SetMaterialAttributes(MaterialAttributes=@n19.MaterialAttributes, Base Color=Multiply(A=@n19.Base Color) {constB=0.5}) {attributeSetTypes [#2]}) {name=Sand-Material, nodeColor=<rgba 1 0 0.9844 1>}
  n20 = VertexColor() {desc=mask cliff to grass transition}
  n21 = NamedRerouteUsage()
  NamedRerouteDeclaration(None=MaterialFunctionCall(Transition Phase=@n20.B, Height Texture=OneMinus(None=@n21)) {materialFunction=&HeightLerp}.Lerp Alpha No Contrast) {name=GravelEdgeColor-Mask, nodeColor=<rgba 1 0 0.9844 1>}
  NamedRerouteDeclaration(None=MaterialFunctionCall!(
    GravelMaterialInput=NamedRerouteUsage()
    Heightmap=NamedRerouteUsage()
    UseUnmodifiedGravelDiffuse=StaticBool() {value=true}
    UnmodifiedGravelDiffuse=@n18.RGB
    BypassBiomeColors
   ) {materialFunction=/Game/Athena/Apollo/Environments/Terrain_Meshes/MaterialFunctions/MF_DirtMudModifier}.MudMaterial) {name=MudMaterial, nodeColor=<rgba 1 0 0.9844 1>}
  n22 = TextureSampleParameter2D() {
   group=Grass
   parameterName=Height Mask
   samplerType=SAMPLERTYPE_LinearColor
   texture=/Game/Environments/Apollo/Sets/Terrain/Textures/T_Apollo_DirtHeight_M
  }
  n23 = ScalarParameter() {defaultValue=1, parameterName=Grass Height Texture Exponent}
  n24 = NamedRerouteUsage()
  NamedRerouteDeclaration(None=Multiply(A=Power(Base=@n22.R, Exp=@n23), B=@n24)) {name=Grass-Height, nodeColor=<rgba 1 0 0.9844 1>}
  NamedRerouteDeclaration(None=MaterialFunctionCall!(CombinedLandscapeLayersMask=@n24, BypassUseRVT, BypassBiomeColors) {materialFunction=/Game/Athena/Apollo/Environments/Landscape/Materials/Biome/CommonMaterials/Mobile/MF_Mobile_Common_Grass}.ForStaticMesh) {name=Grass-Material, nodeColor=<rgba 1 0 0.9844 1>}
  n25 = NamedRerouteUsage()
  n26 = ScalarParameter() {defaultValue=1.968, parameterName=Dirt Height Multiplier}
  n27 = NamedRerouteUsage()
  n28 = ScalarParameter() {defaultValue=1.97, parameterName=Grass Height Multiplier}
  NamedRerouteDeclaration(None=MaterialFunctionCall!(
    Diffuse 1
    Diffuse 2
    Transistion Phase=OneMinus(None=@n20.B)
    Height Texture 1=Multiply(A=@n25, B=@n26)
    Height Texture 2=Multiply(A=@n27, B=@n28)
    Contrast
   ) {materialFunction=/Engine/Functions/Engine_MaterialFunctions02/Texturing/HeightLerpWithTwoHeightMaps}.Lerp Alpha No Contrast) {name=Grass-BlendMask, nodeColor=<rgba 1 0 0.9844 1>}
  NamedRerouteDeclaration(None=Saturate(None=Multiply(A=@n20.B) {constB=4})) {name=GrassEdgeColor-Mask, nodeColor=<rgba 1 0 0.9844 1>}
  NamedRerouteDeclaration(None=@n18.Height) {name=Gravel-Height, nodeColor=<rgba 1 0 0.9844 1>}
  n29 = NamedRerouteUsage()
  n30 = NamedRerouteUsage()
  n31 = VertexColor()
  n32 = NamedRerouteUsage()
  n33 = StaticSwitchParameter(
   True=BlendMaterialAttributes(
    A=NamedRerouteUsage()
    B=StaticSwitchParameter(True=@n29, False=@n30) {desc=Uses Green vertex channel., group=Layers, parameterName="Sand=True-Mud=False", sortPriority=4}
    Alpha=MaterialFunctionCall(Transition Phase=@n31.B, Height Texture=@n32) {materialFunction=&HeightLerp}.Lerp Alpha No Contrast
   )
   False=NamedRerouteUsage()
  ) {desc="You can only paint Mud or Sand where Gravel has been painted!", group=Layers, parameterName=UseMudOrSand, sortPriority=3}
  NamedRerouteDeclaration(None=@n33) {name=Gravel-Mud/Sand, nodeColor=<rgba 1 0 0.9844 1>}
  NamedRerouteUsage()
  Reroute(None=@n13)
  Reroute(None=@n17)
  TextureSampleParameter2D() {
   group=Grass
   parameterName=Height Mask
   samplerType=SAMPLERTYPE_LinearColor
   texture=&T_Apollo_DirtHeight_M
  }
 ]
}
