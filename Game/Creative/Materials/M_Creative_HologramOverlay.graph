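# Translucent hologram overlay material for Creative, usable on skeletal meshes, with a tintable glow.
# A squared fresnel rim, swapped on mobile feature levels for a normal-derivative, depth-scaled outline, adds to normal-map offset edges and a looping bounding-box sweep.
# The sum is tinted and multiplied by scanlines that share the sweep's random time offset.
# A timestamp-driven in/out transition boosts the result tenfold.
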Material /Game/Creative/Materials/M_Creative_HologramOverlay
graph {
 properties {
  bUsedWithMeshDeformer=true
  bUsedWithSkeletalMesh=true
  blendMode=BLEND_Translucent
  pixelDepthOffsetMode=PDOM_Legacy
  usedShadingModels=Default Lit
 }
 nodes [
  n1 = Normalize(VectorInput=Subtract(A=CameraPositionWS(), B=WorldPosition().XYZ))
  n2 = PixelNormalWS()
  n3 = OneMinus(None=Clamp(None=DotProduct(A=@n1, B=@n2)))
  n4 = Multiply(A=Multiply(A=@n3, B=@n3), B=ScalarParameter() {defaultValue=0.5, parameterName=Brightness})
  n5 = CameraVectorWS()
  n6 = Multiply(A=Normalize(VectorInput=PixelNormalWS()), B=ScalarParameter() {defaultValue=30, parameterName=OutlineScale})
  n7 = DDX(Value=@n6)
  n8 = Abs(None=CrossProduct(A=Subtract(A=@n6, B=@n7), B=Add(A=@n7, B=@n6)))
  n9 = DDY(Value=@n6)
  n10 = Abs(None=CrossProduct(A=Subtract(A=@n6, B=@n9), B=Add(A=@n9, B=@n6)))
  n11 = Divide(
   A=DotProduct(A=Add(A=@n8, B=@n10), B=Constant() {r=0.3333})
   B=Multiply(A=PixelDepth().R, B=ScalarParameter() {defaultValue=0.25, parameterName=PixelDepthScale})
  )
  n12 = OneMinus(None=Power(Base=DotProduct(A=@n5, B=@n2), Exp=Saturate(None=@n11)))
  n13 = LinearInterpolate(A=Round(None=@n12), B=@n12, Alpha=ScalarParameter() {defaultValue=0.3, parameterName=FresnelRoundLERP})
  n14 = FeatureLevelSwitch(Default=@n4, ES2=@n13, ES3_1=@n13, SM4_REMOVED=@n4)
  n15 = TextureCoordinate()
  n16 = ScalarParameter() {defaultValue=0.0015, parameterName=Offset U}
  n17 = Subtract(
   A=TextureSampleParameter2D(UVs=Add(A=@n15, B=@n16)) {parameterName=Normals, samplerType=SAMPLERTYPE_Normal, texture=&FlatNormalBlue}.RGB
   B=TextureSampleParameter2D() {parameterName=Normals, samplerType=SAMPLERTYPE_Normal, texture=&FlatNormalBlue}.RGB
  )
  n18 = TextureCoordinate()
  n19 = TextureCoordinate()
  n20 = Add(
   A=Multiply(A=ComponentMask(None=@n17) {r=true}, B=ComponentMask(None=@n18) {r=true})
   B=Multiply(A=ComponentMask(None=@n17) {g=true}, B=ComponentMask(None=@n19) {g=true})
  )
  n21 = Add(A=Time(), B=ScalarParameter() {group=Hologram, parameterName=Random})
  n22 = Multiply(A=@n21, B=ScalarParameter() {defaultValue=0.25, group=Hologram, parameterName=Sweep Speed})
  n23 = Add(
   A=MaterialFunctionCall() {materialFunction=&BoundingBoxBased_0-1_UVW}.G
   B=ConstantBiasScale(None=Frac(None=@n22)) {bias=-0.5, scale=3}
  )
  n24 = OneMinus(None=@n23)
  n25 = Clamp(None=Multiply(A=Multiply(A=@n23, B=@n24)) {constB=4})
  n26 = ScalarParameter() {defaultValue=80, group=Hologram, parameterName=Sweep Exponent}
  n27 = Add(
   A=Multiply(A=Saturate(None=@n20), B=ScalarParameter() {defaultValue=3, group=Hologram, parameterName=Blueprint Edge Brightness})
   B=Multiply(A=Power(Base=@n25, Exp=@n26), B=ScalarParameter() {defaultValue=2, group=Hologram, parameterName=Sweep Brightness})
  )
  n28 = Multiply(
   A=Multiply(A=Add(A=@n14, B=@n27), B=VectorParameter() {group=Hologram, parameterName=Color, defaultValue=<rgba 0.7917 0.0003 0.3871 1>}.RGB)
   B=MaterialFunctionCall!(
    Number of Scanlines
    Scanline Speed
    "Scanlines Thickness (Bias)"
    Scanlines Dark Value
    Scanlines Light Value
    Block Opacity
    UVs
    Time=@n21
    Block Repeat
   ) {materialFunction=/Game/Effects/Fort_Effects/Effects/Atlas/Materials/MF_Atlas_Scanlines_01}.Result
  )
  LinearInterpolate(
   A=@n28
   B=Multiply(A=@n28) {constB=10}
   Alpha=MaterialFunctionCall(
    TimestampA=ScalarParameter() {group=Hologram, parameterName=TimestampA}
    TransitionLength_In=ScalarParameter() {defaultValue=1, group=Hologram, parameterName=TransitionLength_In}
    TimestampB=ScalarParameter() {group=Hologram, parameterName=TimestampB}
    TransitionLength_Out=ScalarParameter() {defaultValue=1, group=Hologram, parameterName=TransitionLength_Out}
    Transition In Only?=StaticBool()
   ) {materialFunction=&MF_Timestamp_InOut_01}.SmoothStep
  )
  MaterialFunctionCall!(In) {materialFunction=/Game/Packages/Fortress_SharedMaterials/Base_Material_Functions/Hologram}
 ]
}
