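# Unlit, two-sided alpha-composite material for a stationary portal hologram in Creative, with constant opacity and a glowing tinted color.
# An edging texture adds bright borders over a gradient built by sphere-masking powered, centered UVs.
# Panned, repeating UV scanlines add extra glow only near the player, using a world-space sphere mask around the PlayerPosition collection parameter.
# The resulting attributes pass through MF_CreativeMoveTool before output.
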Material /Game/Creative/Materials/M_PortalHologram_Stationary
graph {
 properties {
  bUseMaterialAttributes=true
  blendMode=BLEND_AlphaComposite
  pixelDepthOffsetMode=PDOM_Legacy
  shadingModel=MSM_Unlit
  twoSided=true
  usedShadingModels=Unlit
 }
 nodes [
  n1 = Constant() {r=1}
  n2 = LinearInterpolate(A=@n1, Alpha=OneMinus(None=TextureSample() {samplerType=SAMPLERTYPE_Masks, texture=/Game/Creative/Textures/HologramEdging}.R)) {constB=0}
  n3 = VectorParameter() {parameterName=Hologram Color, defaultValue=<rgba 0 0.1372 0.3712 1>}
  n4 = Multiply(A=@n3.RGB, B=ScalarParameter() {defaultValue=30, parameterName=Hologram Edge Glow Power})
  n5 = ConstantBiasScale(None=TextureCoordinate()) {bias=-0.5, scale=2}
  n6 = SphereMask(
   A=Power(Base=Abs(None=@n5), Exp=Constant() {r=10})
   B=ScalarParameter() {parameterName=Center}
   Radius=ScalarParameter() {defaultValue=0.3, parameterName=Gradient Radius}
   Hardness=ScalarParameter() {parameterName=Gradient Hardness}
  )
  n7 = ScalarParameter() {defaultValue=20, parameterName=Hologram Glow Power}
  n8 = Multiply(
   A=Add(A=Multiply(A=@n2, B=@n4), B=LinearInterpolate(A=@n1, Alpha=@n6) {constB=0.2})
   B=LinearInterpolate(A=Multiply(A=@n3.RGB, B=@n7), Alpha=Constant() {r=0.8}) {constB=0}
  )
  n9 = Panner(Coordinate=TextureCoordinate(), Speed=ScalarParameter() {defaultValue=-0.05, parameterName=Scanline Speed})
  n10 = Frac(None=Multiply(A=ComponentMask(None=@n9) {g=true}, B=ScalarParameter() {defaultValue=45, parameterName=ScanlineRepeat}))
  n11 = Constant() {r=0.5}
  n12 = ScalarParameter() {defaultValue=0.12, parameterName=ScanlineRadius}
  n13 = ScalarParameter() {defaultValue=0.01, parameterName=ScanlineHarndess}
  n14 = CollectionParameter() {collection=&FortniteMaterialParameters, parameterName=PlayerPosition}
  n15 = ComponentMask(None=@n14) {b=true, g=true, r=true}
  n16 = WorldPosition()
  n17 = ScalarParameter() {defaultValue=750, parameterName=Scanline Distance}
  n18 = ScalarParameter() {defaultValue=0.3, parameterName=Scanline Distance Feathering}
  n19 = Multiply(
   A=Multiply(
    A=VectorParameter() {parameterName=Hologram Color, defaultValue=<rgba 0.0048 0.1946 0.3712 1>}.RGB
    B=ScalarParameter() {defaultValue=5.5, parameterName=Hologram Scanline Glow Power}
   )
   B=LinearInterpolate(
    A=Constant()
    B=SphereMask(A=@n10, B=@n11, Radius=@n12, Hardness=@n13)
    Alpha=SphereMask(A=@n15, B=@n16.XYZ, Radius=@n17, Hardness=@n18) {attenuationRadius=350, hardnessPercent=80}
   )
  )
  MaterialFunctionCall(MA In=SetMaterialAttributes!(MaterialAttributes, Emissive Color=Add(A=@n8, B=@n19), Opacity=Constant() {r=0.9}) {attributeSetTypes [#1, #2]}) {materialFunction=&MF_CreativeMoveTool}
 ]
}
