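# MF_CreativeColors is a material function for Creative that picks a team color from a lookup texture and applies it through a mask.
# The rounded TeamColor index comes from a scalar parameter or, via static switch, custom primitive data slot 27, unless the Index input overrides it.
# The index times the LUT's texel size gives the sample UV.
# The Mask input blends from white to that color.
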MaterialFunction /Game/Creative/Materials/MF_CreativeColors
graph {
 properties {description=This function uses a mask to overlay color on a material. Currently holds 33 colors.}
 nodes [
  n1 = ScalarParameter() {bUseCustomPrimitiveData=true, group=TeamColor, parameterName=TeamColorCustomPrimitive27, primitiveDataIndex=27, sliderMax=64}
  n2 = StaticSwitchParameter(True=@n1, False=ScalarParameter() {group=TeamColor, parameterName=TeamColor, sliderMax=64}) {group=TeamColor, parameterName=TeamColorUsesCustomPrimitive27}
  n3 = TextureObjectParameter() {group=TeamColor, parameterName=TeamColorsLUT, texture=&/T_CreativeColors}
  n4 = Multiply(
   A=FunctionInput(Preview=Round(None=@n2)) {bUsePreviewValueAsDefault=true, inputName=Index, inputType=FunctionInput_Scalar, sortPriority=1}
   B=TextureProperty(Tex=@n3) {property=TMTM_TexelSize}
  )
  FunctionOutput(None=LinearInterpolate(
    B=TextureSample(UVs=@n4, Tex=@n3).RGB
    Alpha=FunctionInput() {bUsePreviewValueAsDefault=true, inputName=Mask, inputType=FunctionInput_Scalar}
   ) {constA=1})
 ]
}
